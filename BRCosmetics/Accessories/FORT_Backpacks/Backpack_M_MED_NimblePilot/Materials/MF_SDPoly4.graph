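# Material function that returns a signed distance field for a four-sided polygon, exposed to the library under DelMarUI.
# Inputs are the UVs, defaulting to GetUserInterfaceUV's 9-slice UV, and four vertices, defaulting to a centered diamond.
# A Custom HLSL node loops over the edges for the nearest distance, flipping the sign by crossing tests.
# The sign starts at 2.0, so the output is twice the true distance, negative inside.
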MaterialFunction /BRCosmetics/Accessories/FORT_Backpacks/Backpack_M_MED_NimblePilot/Materials/MF_SDPoly4
graph {
 properties {bExposeToLibrary=true, description=Signed distance field for a 4-sided polygon., libraryCategoriesText [DelMarUI]}
 nodes [
  MaterialExpressionFunctionInput_0 = FunctionInput(Preview=MaterialFunctionCall() {materialFunction=&GetUserInterfaceUV}.9-Slice UV) {
   bUsePreviewValueAsDefault=true
   description="Coordinates of the distance field (0,0 is origin)."
   inputName=UVs
   inputType=FunctionInput_Vector2
  }
  MaterialExpressionFunctionInput_1 = FunctionInput() {
   bUsePreviewValueAsDefault=true
   description=The coordinates of the first vertex.
   inputName=Vertex1
   inputType=FunctionInput_Vector2
   sortPriority=1
   previewValue=<0.5 0.25 0 0>
  }
  MaterialExpressionFunctionInput_3 = FunctionInput() {
   bUsePreviewValueAsDefault=true
   description=The coordinates of the second vertex.
   inputName=Vertex2
   inputType=FunctionInput_Vector2
   sortPriority=2
   previewValue=<0.75 0.5 0 0>
  }
  MaterialExpressionFunctionInput_4 = FunctionInput() {
   bUsePreviewValueAsDefault=true
   description=The coordinates of the third vertex.
   inputName=Vertex3
   inputType=FunctionInput_Vector2
   sortPriority=3
   previewValue=<0.5 0.75 0 0>
  }
  MaterialExpressionFunctionInput_5 = FunctionInput() {
   bUsePreviewValueAsDefault=true
   description=The coordinates of the final vertex.
   inputName=Vertex4
   inputType=FunctionInput_Vector2
   sortPriority=4
   previewValue=<0.25 0.5 0 0>
  }
  n1 = Custom!(
   p=@MaterialExpressionFunctionInput_0
   v1=@MaterialExpressionFunctionInput_1
   v2=@MaterialExpressionFunctionInput_3
   v3=@MaterialExpressionFunctionInput_4
   v4=@MaterialExpressionFunctionInput_5
  ) {
   code="float2 v[] = {v1,v2,v3,v4};\r\nconst int N = 4;\r\n\r\nfloat d = dot(p-v[0],p-v[0]);\r\nfloat s = 2.0;\r\nfor( int i=0, j=N-1; i<N; j=i, i++ )\r\n{\r\n    float2 e = v[j] - v[i];\r\n    float2 w = p - v[i];\r\n    float2 b = w - e*clamp( dot(w,e)/dot(e,e), 0.0, 1.0 );\r\n    d = min( d, dot(b,b) );\r\n    float3 c = float3(p.y>=v[i].y,p.y<v[j].y,e.x*w.y>e.y*w.x);\r\n    if( all(c) || all(!c) ) s = -s;\r\n}\r\nreturn s*sqrt(d);"
   outputType=CMOT_Float1
   inputs [
    {
     inputName=p
     input {expression=&self:MaterialExpressionFunctionInput_0}
    }
    {
     inputName=v1
     input {expression=&self:MaterialExpressionFunctionInput_1}
    }
    {
     inputName=v2
     input {expression=&self:MaterialExpressionFunctionInput_3}
    }
    {
     inputName=v3
     input {expression=&self:MaterialExpressionFunctionInput_4}
    }
    {
     inputName=v4
     input {expression=&self:MaterialExpressionFunctionInput_5}
    }
   ]
  }
  FunctionOutput(None=@n1)
 ]
}
