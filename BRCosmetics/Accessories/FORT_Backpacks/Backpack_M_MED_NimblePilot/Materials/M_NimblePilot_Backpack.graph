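# Character material for the Nimble Pilot backpack that adds a reactive dashboard screen between the shared character PreFX and PostFX functions.
# Velocity, Revolution and Kill Counter parameters, gated by Is Reactive?, drive procedural speed bar, circles, nitro bar and blinking indicators, plus a two-digit flipbook counter.
# Camera-offset local-position UVs fake screen depth, with Fresnel and clearcoat normal.
# Masks from the parent function composite the result into emissive.
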Material /BRCosmetics/Accessories/FORT_Backpacks/Backpack_M_MED_NimblePilot/Materials/M_NimblePilot_Backpack
graph {
 properties {
  bHasPixelAnimation=true
  bUseMaterialAttributes=true
  bUsedWithInstancedStaticMeshes=true
  bUsedWithMeshDeformer=true
  bUsedWithMorphTargets=true
  bUsedWithSkeletalMesh=true
  pixelDepthOffsetMode=PDOM_Legacy
  translucencyLightingMode=TLM_SurfacePerPixelLighting
  usedShadingModels=Default Lit
 }
 nodes [
  n1 = NamedRerouteUsage()
  n2 = Power(Base=@n1, Exp=ScalarParameter() {parameterName=Param_6})
  n3 = ScalarParameter() {defaultValue=0.25, parameterName=Param_5}
  Add(
   A=BumpOffset(Coordinate=TextureCoordinate(), Height=Subtract(A=@n2, B=@n3), HeightRatioInput=ScalarParameter() {defaultValue=0.05, parameterName=Param_7})
   B=ComponentMask(None=VectorParameter() {parameterName=Param_1, defaultValue=<rgba -0.0104 -0.0112 0 0>}.RGB) {g=true, r=true}
  )
  n4 = TextureSample() {
   samplerType=SAMPLERTYPE_Normal
   texture=/BRCosmetics/Accessories/FORT_Backpacks/Backpack_M_MED_NimblePilot/Textures/T_NimblePilot_Backpack_Clearcoat_N
  }
  n5 = MaterialFunctionCall(Normal=@n4.RGB, Flatness=ScalarParameter() {defaultValue=0.2, group=Nimble Pilot - Backpack, parameterName=Screen - Bottom Normal Intensity}) {materialFunction=&FlattenNormal}
  ClearCoatNormalCustomOutput!(None=@n5.Result)
  Constant() {r=0.05}
  n6 = ScalarParameter() {defaultValue=17.5, group=Nimble Pilot - Backpack, parameterName=Speed Bar -Amount}
  Divide(A=@n6) {constB=4}
  n7 = NamedRerouteDeclaration(None=ScalarParameter() {defaultValue=4, group=Nimble Pilot - Backpack, parameterName=Flipbook - Rows}) {name=Rows, nodeColor=<rgba 1 0 0.9844 1>}
  n8 = NamedRerouteDeclaration(None=ScalarParameter() {defaultValue=3, group=Nimble Pilot - Backpack, parameterName=Flipbook - Columns}) {name=Columns, nodeColor=<rgba 0 1 0.6953 1>}
  n9 = Subtract(
   A=Divide(B=Multiply(A=@n7, B=@n8)) {desc=Normalized amount of slots for kill counter increment}
   B=Constant() {desc=Truncate last digit for accuracy, r=0}
  )
  n10 = LinearInterpolate(
   B=ScalarParameter() {defaultValue=56, group=Nimble Pilot - Backpack, parameterName=Kill Counter}
   Alpha=ScalarParameter() {defaultValue=1, group=Nimble Pilot - Backpack - Reactivity, parameterName=Is Reactive?}
  ) {constA=-1}
  n11 = Clamp(None=Floor(None=@n10)) {maxDefault=99}
  n12 = Constant() {desc=Divided and mod by 10 as its a two digit number, r=10}
  n13 = Constant() {desc=11 is out the number range so it shows no number if its single digit, r=11}
  n14 = If(A=@n11, "A > B"=Modulo!(A=@n11, B=@n12), "A == B"=@n13, "A < B"=@n13)
  n15 = Clamp(None=Add(A=@n9, B=Multiply(A=@n9, B=@n14))) {maxDefault=1}
  Frac(None=@n15)
  n16 = MaterialFunctionCall() {materialFunction=&MF_CharacterParent_2023_PreFX}
  n17 = GetMaterialAttributes(None=@n16.Result) {attributeGetTypes [
    #1
    #2
    #3
    #4
   ]}
  n18 = SetMaterialAttributes!(
   MaterialAttributes=@n17.MaterialAttributes
   Emissive Color=NamedRerouteUsage()
   Base Color=NamedRerouteUsage()
   Normal
   Shading Model
   Custom Data 0
   Custom Data 1
   Roughness
   World Position Offset=NamedRerouteUsage()
  ) {attributeSetTypes [
    #1
    #2
    #5
    #6
    #7
    #8
    #3
    #9
   ]}
  MaterialFunctionCall(
   Material=BlendMaterialAttributes(A=@n17.MaterialAttributes, B=@n18, Alpha=NamedRerouteUsage())
   M=@n16.M
   Pre Skinned Local Position=@n16.M
   Distance Blend=@n16.M
   Fresnel=@n16.M
  ) {materialFunction=&MF_CharacterParent_2023_PostFX}
  n19 = Floor(None=Divide(A=@n11, B=@n12) {constB=10})
  n20 = If(A=@n10, "A > B"=@n19, "A == B"=@n19, "A < B"=@n13) {constB=10}
  n21 = Clamp(None=Add(A=@n9, B=Multiply(A=@n9, B=@n20))) {maxDefault=1}
  n22 = NamedRerouteUsage()
  n23 = NamedRerouteUsage()
  n24 = TextureObjectParameter() {
   group=Nimble Pilot - Backpack
   parameterName=Number Decals
   texture=/BRCosmetics/Accessories/FORT_Backpacks/Backpack_M_MED_NimblePilot/Textures/T_NimblePilot_Number_Decals
  }
  n25 = NamedRerouteUsage()
  n26 = ScalarParameter() {defaultValue=-10, parameterName=Flipbook Mip Bias}
  MaterialFunctionCall("Animation  Phase (0-1)"=@n21, Number of Rows=@n22, Number of Columns=@n23, Texture=@n24, UVs=@n25, MipBias/Level=@n26) {materialFunction=&FlipBook}
  n27 = NamedRerouteUsage()
  n28 = NamedRerouteUsage()
  n29 = TextureObjectParameter() {
   group=Nimble Pilot - Backpack
   parameterName=Number Decals
   texture=&T_NimblePilot_Number_Decals
  }
  n30 = VectorParameter() {group=Nimble Pilot - Backpack, parameterName=Kill Counter - Padding, defaultValue=<rgba 0.394 0 0 0>}
  n31 = VectorParameter() {group=Nimble Pilot - Backpack - Reactivity, parameterName="Kill Counter - Double Digit -  Position", defaultValue=<rgba 0 0.608 0 0>}
  n32 = LinearInterpolate(
   A=VectorParameter() {group=Nimble Pilot - Backpack, parameterName=Kill Counter - Position, defaultValue=<rgba 0 0.608 0 0>}.RGB
   B=@n31.RGB
   Alpha=LinearInterpolate(
    B=ScalarParameter() {defaultValue=1, group=Nimble Pilot - Backpack - Reactivity, parameterName=Double Digits}
    Alpha=ScalarParameter() {defaultValue=1, group=Nimble Pilot - Backpack - Reactivity, parameterName=Is Reactive?}
   )
  )
  n33 = Add(A=Multiply(A=@n30.RGB) {constB=-1}, B=@n32)
  n34 = NamedRerouteUsage()
  n35 = ComponentMask(None=VectorParameter() {group=Nimble Pilot - Backpack, parameterName=Kill Counter - Scale}.RGB) {g=true, r=true}
  n36 = MaterialFunctionCall(
   UVs=Multiply(A=Multiply(A=@n34, B=@n35), B=ScalarParameter() {defaultValue=1, group=Nimble Pilot - Backpack, parameterName=Kill Counter - Scale})
   "Rotation Angle (0-1)"=ScalarParameter() {defaultValue=0.25, group=Nimble Pilot - Backpack, parameterName=Rotation}
  ) {materialFunction=&CustomRotator}
  n37 = NamedRerouteDeclaration(None=Add(A=ComponentMask(None=@n33) {g=true, r=true}, B=@n36.Rotated Values)) {name=Digit 02 UV, nodeColor=<rgba 0 0.9062 1 1>}
  MaterialFunctionCall("Animation  Phase (0-1)"=@n15, Number of Rows=@n27, Number of Columns=@n28, Texture=@n29, UVs=@n37, MipBias/Level=@n26) {materialFunction=&FlipBook}
  n38 = NamedRerouteUsage()
  n39 = NamedRerouteUsage()
  n40 = VectorParameter() {group=Nimble Pilot - Backpack, parameterName=Speed Bar - Colour, defaultValue=<rgba 0.9734 0.3185 0.1221 1>}
  n41 = LinearInterpolate(
   A=Multiply(A=ScalarParameter() {defaultValue=0.4, group=Nimble Pilot - Backpack, parameterName=Default Speed Opacity}, B=@n38)
   B=Multiply(A=@n39, B=@n40.RGB)
   Alpha=@n39
  )
  n42 = VectorParameter() {group=Nimble Pilot - Backpack, parameterName=Speed Bar - Leading Bar Colour, defaultValue=<rgba 1.2 1.0123 0.1069 1>}
  n43 = Multiply(A=@n38, B=@n42.RGB)
  n44 = NamedRerouteUsage()
  n45 = VectorParameter() {group=Nimble Pilot - Backpack, parameterName=Speed Circle - Colour, defaultValue=<rgba 0.0052 0.9473 0.9387 1>}
  n46 = Max(A=LinearInterpolate(A=LinearInterpolate(A=@n41, B=@n43, Alpha=@n44), B=@n45.RGB, Alpha=NamedRerouteUsage()), B=NamedRerouteUsage())
  Multiply(
   A=Add(A=@n46, B=NamedRerouteUsage())
   B=Multiply(A=NamedRerouteUsage(), B=ScalarParameter() {defaultValue=1, group=Nimble Pilot - Backpack, parameterName=Screen Depth - Fade Intensity})
  )
  n47 = MaterialFunctionCall(UVs=NamedRerouteUsage(), "Rotation Angle (0-1)"=ScalarParameter() {defaultValue=0.25, group=Nimble Pilot - Backpack, parameterName=Rotation}) {materialFunction=&CustomRotator}
  n48 = ScalarParameter() {defaultValue=0.6596, group=Nimble Pilot - Backpack, parameterName=Speed Circle - Scale}
  Multiply(A=@n47.Rotated Values, B=@n48)
  n49 = VectorParameter() {group=Nimble Pilot - Backpack, parameterName=Graphics - Global Offset, defaultValue=<rgba -0.0104 -0.0112 0 0>}
  NamedRerouteDeclaration(None=Add(A=NamedRerouteUsage(), B=ComponentMask(None=@n49.RGB) {g=true, r=true})) {name=Depth UVs, nodeColor=<rgba 0.4766 0 1 1>}
  NamedRerouteDeclaration(None=ComponentMask(None=@n16.Tech Art Mask) {b=true}) {name=Screen Mask, nodeColor=<rgba 0 0.7187 1 1>}
  n50 = TextureSample() {
   samplerType=SAMPLERTYPE_Masks
   texture=/BRCosmetics/Accessories/FORT_Backpacks/Backpack_M_MED_NimblePilot/Textures/T_NimblePilot_Backpack_FX2
  }
  NamedRerouteDeclaration(None=@n50.R) {name=Screen Depth, nodeColor=<rgba 0 0.8828 1 1>}
  NamedRerouteDeclaration(None=@n17.Base Color) {name=Base Color, nodeColor=<rgba 1 0 0.1875 1>}
  n51 = NamedRerouteUsage()
  n52 = ScalarParameter() {group=Nimble Pilot - Backpack, parameterName=Screen Fresnel - Color Intensity}
  n53 = Power(Base=NamedRerouteUsage(), Exp=ScalarParameter() {defaultValue=6.9044, group=Nimble Pilot - Backpack, parameterName=Screen Depth - Color Exp})
  NamedRerouteDeclaration(None=LinearInterpolate(B=Multiply(A=@n51, B=@n52), Alpha=Saturate(None=@n53)) {constB=0}) {name=Base Color Out, nodeColor=<rgba 0 0.5078 1 1>}
  n54 = Fresnel(
   ExponentIn=ScalarParameter() {defaultValue=0.036, group=Nimble Pilot - Backpack, parameterName=Screen Fresnel - Emissive Exp}
   Normal=Transform(None=@n5.Result)
  )
  NamedRerouteDeclaration(None=Multiply(
    A=OneMinus(None=@n54)
    B=ScalarParameter() {defaultValue=0.3, group=Nimble Pilot - Backpack, parameterName=Screen Fresnel - Emissive Intensity}
   )) {name=Screen Fresnell, nodeColor=<rgba 1 0 0.8437 1>}
  n55 = VectorParameter() {group=Nimble Pilot - Backpack, parameterName=Depth - Camera Pos, defaultValue=<rgba -69.6169 0 -2156.7671 0>}
  n56 = Add(A=CameraPositionWS(), B=@n55.RGB)
  n57 = ScalarParameter() {defaultValue=0.005, group=Nimble Pilot - Backpack, parameterName=Depth}
  n58 = OneMinus(None=NamedRerouteUsage())
  n59 = Subtract(
   A=VertexInterpolator(VS=LocalPosition() {localOrigin=InstancePreSkinning}.XYZ).PS
   B=Multiply(A=Transform(None=@n56) {transformSourceType=TRANSFORMSOURCE_World, transformType=TRANSFORM_Local}, B=Multiply(A=@n57, B=@n58))
  )
  n60 = Constant() {desc=Scale Local Position to UV Scale, r=0.1}
  NamedRerouteDeclaration(None=ComponentMask(None=Multiply(A=@n59, B=@n60)) {b=true, r=true}) {name=Camera Bump Offset, nodeColor=<rgba 0 1 0.1094 1>}
  NamedRerouteDeclaration(None=Constant()) {name=TSR Velocity, nodeColor=<rgba 0 0.4375 1 1>}
  NamedRerouteDeclaration(None=@n17.Base Color) {name=Normal In, nodeColor=<rgba 1 0.2578 0 1>}
  n61 = NamedRerouteUsage()
  n62 = Step(X=@n61) {constY=0.001}
  n63 = NamedRerouteUsage()
  n64 = LinearInterpolate(
   A=LinearInterpolate(Alpha=Ceil(None=@n63)) {constB=0.15}
   B=Power(Base=@n63, Exp=ScalarParameter() {defaultValue=1, parameterName=Speed Indicator - Velocity Power})
   Alpha=@n63
  ) {constA=0.2}
  n65 = Step(Y=@n61, X=Clamp(None=@n64) {maxDefault=0.99})
  n66 = Constant() {desc=Speed Indicator - Selects the Blinking Light Width, r=0.13}
  n67 = Subtract(A=@n64, B=@n66)
  n68 = Add(A=@n66, B=@n64) {constB=0.2}
  n69 = Min(A=Step(Y=Clamp(None=@n67) {minDefault=0.2}, X=@n61), B=Step(Y=@n61, X=Clamp(None=@n68) {minDefault=0.2}))
  n70 = MaterialFunctionCall!(
   Sine Phase=Multiply(A=Time(), B=ScalarParameter() {defaultValue=0.4, group=Nimble Pilot - Backpack, parameterName=Speed Indicator - Blinking Light - Speed})
   Value 1
   Value 2
  ) {materialFunction=/Engine/Functions/Engine_MaterialFunctions02/Utility/Sine_Remapped}
  NamedRerouteDeclaration(None=Max(A=Multiply(A=@n62, B=@n65), B=Multiply(A=@n69, B=@n70.Remapped Sine) {constB=0.5})) {name=Speed Indicatiors, nodeColor=<rgba 1 0.6563 0 1>}
  NamedRerouteDeclaration(None=@n16.Tech Art Mask Alpha) {name=Speed Indicator Mask, nodeColor=<rgba 1 0 0.0234 1>}
  n71 = Subtract(
   A=MaterialFunctionCall(
    UVs=NamedRerouteUsage()
    Texture Scale=ScalarParameter() {defaultValue=0.225, group=Nimble Pilot - Backpack, parameterName=Nitro Bar - Scale}
   ) {materialFunction=&ScaleUVsByCenter}.UVs
   B=ComponentMask(None=VectorParameter() {group=Nimble Pilot - Backpack, parameterName=Nitro Bar - Position, defaultValue=<rgba -3 0.5608 0 0>}.RGB) {g=true, r=true}
  )
  n72 = Constant3Vector() {constant=<rgba 0.3 0.4 0 0>}
  n73 = VectorParameter() {group=Nimble Pilot - Backpack, parameterName=Nitro Bar - Width, defaultValue=<rgba 1.36 0 0 0>}
  n74 = Constant3Vector() {constant=<rgba 0.4 0.6 0 0>}
  n75 = MaterialFunctionCall!(UVs=@n71, Vertex1=@n72, Vertex2=Add(A=@n72, B=@n73.RGB), Vertex3=Add(A=@n73.RGB, B=@n74), Vertex4=@n74) {materialFunction=&/MF_SDPoly4}
  n76 = Constant() {r=1}
  n77 = MaterialFunctionCall(UVs=TextureCoordinate(), "Rotation Angle (0-1)"=ScalarParameter() {defaultValue=0.068}) {materialFunction=&CustomRotator}
  n78 = ComponentMask(None=VectorParameter() {defaultValue=<rgba 3.1192 81.6327 0 1>}.RGB) {g=true, r=true}
  n79 = MaterialFunctionCall!(
   Width=ScalarParameter() {defaultValue=0.0131, parameterName=Param_1}
   Sharpness=@n76
   Offset=ScalarParameter() {defaultValue=1.174, parameterName=Param_2}
   Direction Switch
   Compare
   Input Coordinates=@n77.Rotated Values
  ) {materialFunction=/Engine/Functions/Engine_MaterialFunctions03/Procedurals/GeneratedBand}
  n80 = Multiply(
   A=MaterialFunctionCall(
    Width=ScalarParameter() {defaultValue=0.1642, parameterName=Param_3}
    Sharpness=@n76
    Offset=ScalarParameter() {defaultValue=4488.9419, parameterName=Param_4}
    Input Coordinates=Multiply(A=@n77.Rotated Values, B=@n78) {constB=2}
   ) {materialFunction=&GeneratedBand}.Result
   B=@n79.Result
  )
  n81 = Saturate(None=Subtract(A=Step(Y=@n75.Result) {constX=0.001}, B=@n80))
  n82 = ScalarParameter() {defaultValue=0.25, group=Nimble Pilot - Backpack, parameterName=Nitro Bar - Default Value}
  n83 = NamedRerouteUsage()
  n84 = ScalarParameter() {defaultValue=1.3, parameterName=Nitro Bar - Velocity Power}
  n85 = LinearInterpolate(
   B=VectorParameter() {group=Nimble Pilot - Backpack, parameterName=Nitro Bar - Width, defaultValue=<rgba 1.36 0 0 0>}.RGB
   Alpha=Saturate(None=Power(Base=@n83, Exp=@n84))
  ) {constA=-0.0001}
  n86 = Step(Y=MaterialFunctionCall(UVs=@n71, Vertex1=@n72, Vertex2=Add(A=@n72, B=@n85), Vertex3=Add(A=@n74, B=@n85), Vertex4=@n74) {materialFunction=&MF_SDPoly4}.Result) {constX=0.001}
  n87 = Ceil(None=NamedRerouteUsage())
  n88 = Saturate(None=Subtract(A=LinearInterpolate(B=@n86, Alpha=@n87), B=@n80))
  n89 = VectorParameter() {group=Nimble Pilot - Backpack, parameterName=Nitro Bar - Red, defaultValue=<rgba 0.7084 0.0242 0.0116 1>}
  n90 = Multiply(A=@n88, B=@n89.RGB)
  NamedRerouteDeclaration(None=LinearInterpolate(A=Multiply(A=@n81, B=@n82), B=LinearInterpolate(A=@n88, B=@n90, Alpha=@n79.Result), Alpha=@n88)) {name=Nitro Bar, nodeColor=<rgba 0 0.4375 1 1>}
  n91 = VectorParameter() {group=Nimble Pilot - Backpack, parameterName=Speed Bar - Position}
  n92 = MaterialFunctionCall(
   UVs=Add(A=NamedRerouteUsage(), B=ComponentMask(None=@n91.RGB) {g=true, r=true})
   Texture Scale=ScalarParameter() {defaultValue=1, group=Nimble Pilot - Backpack, parameterName=Speed Bar - Scale}
  ) {materialFunction=&ScaleUVsByCenter}
  n93 = ScalarParameter() {defaultValue=0.77, group=Nimble Pilot - Backpack, parameterName=Speed Bar - Rotation}
  n94 = MaterialFunctionCall(Vector or UVs=Subtract(A=MaterialFunctionCall(UVs=@n92.UVs, "Rotation Angle (0-1)"=@n93) {materialFunction=&CustomRotator}.Rotated Values) {constB=0.5}) {materialFunction=&VectorToRadialValue}
  n95 = Abs(None=@n94.Vector Converted to Angle)
  n96 = LinearInterpolate(
   A=ScalarParameter() {group=Nimble Pilot - Backpack - Reactivity, parameterName=Revolution - Non Reactive, sliderMax=1}
   B=ScalarParameter() {group=Nimble Pilot - Backpack - Reactivity, parameterName=Revolution}
   Alpha=ScalarParameter() {defaultValue=1, group=Nimble Pilot - Backpack - Reactivity, parameterName=Is Reactive?}
  )
  n97 = ScalarParameter() {defaultValue=2, parameterName=Revolution - Velocity Power}
  n98 = LinearInterpolate(Alpha=Saturate(None=Power(Base=@n96, Exp=@n97))) {constB=0.27}
  n99 = Subtract(A=@n98, B=ScalarParameter() {defaultValue=0.044, parameterName=Param_30})
  NamedRerouteDeclaration(None=Min(A=Step(Y=@n95, X=@n98), B=Step(Y=@n99, X=@n95))) {name=Leading Bar, nodeColor=<rgba 1 0.9844 0 1>}
  n100 = Frac(None=AppendVector(A=Multiply(A=@n6, B=@n94.Vector Converted to Angle) {constB=16}, B=Multiply(A=@n94.Linear Distance) {constB=2}))
  n101 = VectorParameter() {group=Nimble Pilot - Backpack, parameterName=Speed Bar - Size, defaultValue=<rgba 0.644 0.5 0 0>}
  n102 = Multiply(
   A=Step(Y=MaterialFunctionCall!(UVs=@n100, Size=@n101.RGB, CornerRoundness, KeepAspectRatio, PixelSize) {materialFunction=/FortUILibrary/MaterialFunctions/Shapes/SDFs/MF_SDF_Box}.SDF) {constX=0.01}
   B=MaterialFunctionCall(
    Width=ScalarParameter() {defaultValue=0.072, group=Nimble Pilot - Backpack, parameterName=Speed Bar - Radius}
    Sharpness=Constant() {r=1}
    Offset=Constant() {r=0.1168}
    Input Coordinates=@n94.Radial Coordinates
   ) {materialFunction=&GeneratedBand}.Result
  )
  NamedRerouteDeclaration(None=Multiply(A=@n102, B=Step(Y=@n95) {constX=0.27})) {name=Default Speed Bar, nodeColor=<rgba 0 0.0859 1 1>}
  n103 = Add(A=@n30.RGB, B=@n32)
  NamedRerouteDeclaration(None=Add(A=@n36.Rotated Values, B=ComponentMask(None=@n103) {g=true, r=true})) {name=Digit 01 UV, nodeColor=<rgba 1 0 0.5859 1>}
  n104 = MaterialFunctionCall!(
   "Animation  Phase (0-1)"=@n21
   Number of Rows=@n22
   Number of Columns=@n23
   Texture=@n24
   UVs=@n25
   MipBias/Level=@n26
   "Use Mip Bias (T) Level (F)"
   "Clamp Anim (see tooltip)"
  ) {materialFunction=/BRCosmetics/Characters/Player/Female/Medium/Bodies/F_MED_ZombieElasticNeon/FX/FlipBook_Clamped}
  n105 = Constant() {desc=Number Clipping, r=0.76}
  n106 = Step(Y=MaterialFunctionCall(UVs=NamedRerouteUsage(), Size=@n105) {materialFunction=&MF_SDF_Box}.SDF) {constX=0.01}
  n107 = MaterialFunctionCall("Animation  Phase (0-1)"=@n15, Number of Rows=@n27, Number of Columns=@n28, Texture=@n29, UVs=@n37, MipBias/Level=@n26) {materialFunction=&FlipBook_Clamped}
  n108 = Step(Y=MaterialFunctionCall(UVs=NamedRerouteUsage(), Size=@n105) {materialFunction=&MF_SDF_Box}.SDF) {constX=0.01}
  NamedRerouteDeclaration(None=Max(A=Multiply(A=@n104.Alpha, B=@n106), B=Multiply(A=@n107.Alpha, B=@n108))) {name=Kill Counter, nodeColor=<rgba 1 0.7734 0 1>}
  n109 = Subtract(
   A=MaterialFunctionCall(UVs=@n47.Rotated Values, Texture Scale=@n48) {materialFunction=&ScaleUVsByCenter}.UVs
   B=ComponentMask(None=VectorParameter() {group=Nimble Pilot - Backpack, parameterName=Speed Circle - Position}.RGB) {g=true, r=true}
  )
  n110 = Multiply(A=@n109) {constB=3}
  n111 = Step(Y=MaterialFunctionCall!(UVs=Frac(None=@n110), Radius) {materialFunction=/FortUILibrary/MaterialFunctions/Shapes/SDFs/MF_SDF_Circle}.SDF) {constX=0.1}
  n112 = Step(Y=MaterialFunctionCall(UVs=@n109, Size=VectorParameter() {group=Nimble Pilot - Backpack, parameterName=Speed Circle Mask, defaultValue=<rgba 1 0.396 0 0>}.RGB) {materialFunction=&MF_SDF_Box}.SDF) {constX=0.001}
  n113 = ComponentMask(None=@n109) {r=true}
  n114 = Multiply(A=Floor(None=Multiply(A=@n113) {constB=3}), B=Divide() {constB=3}) {constB=3}
  n115 = NamedRerouteDeclaration(None=LinearInterpolate(
    A=ScalarParameter() {group=Nimble Pilot - Backpack - Reactivity, parameterName=Velocity - Non Reactive, sliderMax=1}
    B=ScalarParameter() {group=Nimble Pilot - Backpack - Reactivity, parameterName=Velocity, sliderMax=1}
    Alpha=ScalarParameter() {defaultValue=1, group=Nimble Pilot - Backpack - Reactivity, parameterName=Is Reactive?}
   )) {name=Velocity, nodeColor=<rgba 0.9219 0 1 1>}
  n116 = ScalarParameter() {defaultValue=2.5, parameterName=Speed Circles - Velocity Power}
  n117 = LinearInterpolate(Alpha=Saturate(None=Power(Base=@n115, Exp=@n116))) {constA=-0.001}
  NamedRerouteDeclaration(None=Multiply(A=Multiply(A=@n111, B=@n112), B=Step(Y=@n114, X=@n117))) {name=Speed Circles, nodeColor=<rgba 0 1 0.5313 1>}
  NamedRerouteDeclaration(None=Multiply(A=@n102, B=Step(Y=@n95, X=@n98))) {name=Speed Bar, nodeColor=<rgba 0.0312 1 0 1>}
  NamedRerouteDeclaration(None=@n17.Emissive Color) {name=Emissive Color In, nodeColor=<rgba 1 0 0.3516 1>}
  n118 = VectorParameter() {group=Nimble Pilot - Backpack, parameterName=Speed Indicator - Colour, defaultValue=<rgba 0.7682 0.6445 0.0648 1>}
  n119 = OneMinus(None=Fresnel(ExponentIn=ScalarParameter() {defaultValue=0.5, parameterName=Speed Indicator - Fresnel}) {baseReflectFraction=0})
  n120 = Add(A=NamedRerouteUsage(), B=LinearInterpolate(A=@n46, B=Multiply(A=@n118.RGB, B=@n119), Alpha=NamedRerouteUsage()))
  n121 = ScalarParameter() {defaultValue=0.7, group=Nimble Pilot - Backpack, parameterName=Screen Opacity}
  NamedRerouteDeclaration(None=LinearInterpolate(A=NamedRerouteUsage(), B=Multiply(A=@n120, B=@n121), Alpha=NamedRerouteUsage())) {name=Emissive Out, nodeColor=<rgba 0 0.6719 1 1>}
  OneMinus(None=@n1)
  PixelNormalWS()
  Saturate(None=@n2)
  ScalarParameter() {defaultValue=0.135, parameterName=Param_29}
  ScalarParameter() {defaultValue=56, group=Nimble Pilot - Backpack, parameterName=Kill Counter}
  n122 = MaterialFunctionCall(Float2=PreviousFrameSwitch(Current Frame=Constant(), Previous Frame=@n21)) {materialFunction=&BreakOutFloat2Components}
  Transform(None=MaterialFunctionCall(X=@n122.R, Y=@n122.R) {materialFunction=&MakeFloat3}.Result)
 ]
}
