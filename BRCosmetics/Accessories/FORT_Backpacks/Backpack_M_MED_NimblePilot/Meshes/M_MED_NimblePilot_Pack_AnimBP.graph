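# Anim Blueprint for the NimblePilot backpack that turns its steering wheel bone as the owning pawn moves.
# Each update derives wheel roll from speed and movement azimuth, flips it when moving backward, fades it near straight back, and interpolates from the bone's current roll.
# The AnimGraph adds a fixed offset, then that rotation, to dyn_steeringWheel.
# Initialization and customization events set character-part-dependent backpack offsets, which the AnimGraph does not read.
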AnimBlueprint /BRCosmetics/Accessories/FORT_Backpacks/Backpack_M_MED_NimblePilot/Meshes/M_MED_NimblePilot_Pack_AnimBP
target_skeleton=&/M_MED_NimblePilot_Pack_Skeleton
graphs [
 AnimGraph {nodes [
   n1 = "Transform (Modify) Bone - Bone: dyn_steeringWheel"(ComponentPose=Local To Component(LocalPose=Input Pose()), Alpha=1, Translation="2.3,0,3.5", Rotation="0,0,10", Scale="1,1,1") {
    rotationMode=BMM_Additive
    translationMode=BMM_Additive
    translationSpace=BCS_BoneSpace
    alphaBoolBlend {blendOption=Linear}
    boneToModify {boneName=dyn_steeringWheel}
    rotation {roll=10}
    translation {x=2.3, z=3.5}
   }
   n2 = "Transform (Modify) Bone - Bone: dyn_steeringWheel"(ComponentPose=@n1, Alpha=1, Translation="0,0,0", Rotation=Get SteeringWheelRot() -> (SteeringWheelRot:Rotator), Scale="1,1,1") {rotationMode=BMM_Additive, rotationSpace=BCS_BoneSpace, alphaBoolBlend {blendOption=Linear}, boneToModify {boneName=dyn_steeringWheel}}
   Output Pose(Result=Component To Local(ComponentPose=@n2)) {layerGroup=None}
  ]}
 EventGraph {nodes [
   Event BlueprintUpdateAnimation()
   n1 = GetPartSkeletalMeshComponent(self=Self-Reference())
   n2 = Get PawnOwner() -> (PawnOwner:Pawn Object Reference)
   n3 = MapRangeClamped(
    Value=Dot Product!(A:Vector=Get PawnVelocity(), B:Vector=GetActorForwardVector!(self:Actor Object Reference=@n2) -> (ReturnValue:Vector)) -> (ReturnValue:double)
    InRangeA=0
    InRangeB=-200
    OutRangeA=1
    OutRangeB=-1
   )
   n4 = Get PawnMovementDirectionAzimuth() -> (PawnMovementDirectionAzimuth:float)
   n5 = Get PawnSpeed2D()
   n6 = float * float(
    A=float * float(
     A=MapRangeClamped(Value=@n5, InRangeA=200, InRangeB=770)
     B=MapRangeClamped(Value=@n4, InRangeA=-90, InRangeB=90, OutRangeA=75, OutRangeB=-75)
    )
    B=float * float(A=MapRangeClamped(Value=@n4, InRangeA=179, InRangeB=170), B=MapRangeClamped(Value=@n4, InRangeA=-179, InRangeB=-170))
   ) {is=title K2Node_PromotableOperator}
   > Set!(
    execute
    SteeringWheelRot_Roll:float=FInterpTo(
     Current=GetSocketTransform!(self=@n1, InSocketName=dyn_steeringWheel, TransformSpace=RTS_ParentBoneSpace) -> (
      ReturnValue_Location:Vector
      ReturnValue_Rotation_Roll:float
      ReturnValue_Rotation_Pitch:float
      ReturnValue_Rotation_Yaw:float
      ReturnValue_Scale:Vector
     ).ReturnValue_Rotation_Roll
     Target=float * float(A=@n6, B=@n3)
     DeltaTime=GetWorldDeltaSeconds!() -> (ReturnValue:double)
     InterpSpeed=10
    )
    SteeringWheelRot_Pitch:float=0
    SteeringWheelRot_Yaw:float=0
   ) -> (then, Output_Get:Rotator)
   n7 = Event BlueprintInitializeAnimation()
   n8 = Set PawnOwner(execute=[@n7.then, @n9.then], PawnOwner:Pawn Object Reference=TryGetPawnOwner()) -> (Output_Get:Pawn Object Reference)
   n9 = Event OnCharacterCustomizationLoaded()
   n10 = Get Array Item with EFort Custom Part Type(Array:Array of Custom Character Part Object References=@n9.CustomParts) -> (Output:Custom Character Part Object Reference)
   n11 = Get Array Item with EFort Custom Part Type(Array=@n9.CustomParts, PartType=Head)
   n12 = Contains(SearchIn=Append(A=GetObjectName(Object=@n11), B=GetObjectName(Object=@n10)), Substring=NimblePilot)
   Set BackpackLocOffset(execute=@n8.then, BackpackLocOffset:Vector=Select(Option 0:Vector="2,0,0", Option 1:Vector="2.3,0,3", Index=@n12) -> (ReturnValue:Vector)) -> (Output_Get:Vector)
   > Set BackpackRotOffset(BackpackRotOffset:Rotator=Select(Option 0:Rotator="0,0,-5", Option 1:Rotator="0,0,10", Index=@n12) -> (ReturnValue:Rotator)) -> (Output_Get:Rotator)
  ]}
]
variables [
 BackpackLocOffset {type=Vector, default=<2.5 0 1>}
 BackpackRotOffset {type=Rotator, default=<rot 0 0 12>}
 PawnOwner {type=Pawn, default=null}
 SteeringWheelRot {type=Rotator}
]
